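# Physics asset for the female medium Spy Tech player body, covering its dangling accessories: pouches, straps and gloves.
# Capsules on the pelvis and thighs form the body volumes, with small base bodies linked to them by angularly free constraints.
# Accessory bodies chain off these bases with limited swing and tight twist: multi-segment back and front belt straps, a right-leg pouch with strap, a back pouch, and two pelvis-anchored gloves.
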PhysicsAsset /BRCosmetics/Characters/Player/Female/Medium/Bodies/F_MED_Spy_Tech/Meshes/F_MED_Spy_Tech_Physics
bodies {
 dyn_back_pouch [{shapeType=Capsule, length=8.25, radius=2.75, center=<3.0618 0.0001 3.4663>, rotation=<rot 90 -179.9802 -179.9802>}]
 dyn_back_pouch_base [{shapeType=Capsule, length=1, radius=1}]
 dyn_bk_strapA_a [{shapeType=Box, extentX=10.5, extentY=1.5, extentZ=4.5, center=<5.1156 0 0>}]
 dyn_bk_strapA_a_base [{shapeType=Capsule, length=1, radius=1}]
 dyn_bk_strapA_b [{shapeType=Box, extentX=7, extentY=1.5, extentZ=4.5, center=<3.7149 0 0>}]
 dyn_bk_strapA_c [{shapeType=Box, extentX=9, extentY=1.5, extentZ=4.5, center=<4.6764 0 0>}]
 dyn_bk_strapB_a [{shapeType=Box, extentX=8.5, extentY=1.5, extentZ=5, center=<5.5198 0 0>}]
 dyn_bk_strapB_a_base [{shapeType=Box, extentX=1.01, extentY=1.01, extentZ=1.01, rotation=<rot 0 -90 0>}]
 dyn_bk_strapB_b [{shapeType=Box, extentX=7.5, extentY=1.5, extentZ=5, center=<4.0134 0 0.0001>}]
 dyn_bk_strapB_c [{shapeType=Box, extentX=10.5, extentY=1.5, extentZ=5, center=<5.4443 0 -0.0001>}]
 dyn_fr_beltStrapA_base_l [{shapeType=Capsule, length=1, radius=1}]
 dyn_fr_beltStrapA_l [{shapeType=Box, extentX=4.5, extentY=1, extentZ=4.5, center=<2.5609 0 0>}]
 dyn_fr_beltStrapB_l [{shapeType=Box, extentX=5, extentY=1, extentZ=4.5, center=<2.5715 0 0>}]
 dyn_fr_beltStrapC_l [{shapeType=Box, extentX=5, extentY=1, extentZ=3.5, center=<2.4342 0.3749 0.075>, rotation=<rot 0 5.3019 0>}]
 dyn_glove_base_l [{shapeType=Capsule, length=1, radius=1}]
 dyn_glove_base_r [{shapeType=Capsule, length=1, radius=1, rotation=<rot 0 0 180>}]
 dyn_glove_l [{shapeType=Capsule, length=2.5, radius=2.75, center=<-1.9274 0 0>, rotation=<rot 90 0 0>}]
 dyn_glove_r [{shapeType=Capsule, length=2.5, radius=2.75, center=<1.9274 0 0>, rotation=<rot 90 0 180>}]
 dyn_leg_pouchStrap_r [{shapeType=Capsule, length=7.75, radius=2, center=<5.323 0.0002 0>, rotation=<rot 90 0 0>}]
 dyn_leg_pouch_base_r [{shapeType=Capsule, length=1, radius=1}]
 dyn_leg_pouch_r [{shapeType=Capsule, length=6.75, radius=3.75, center=<1.6238 0 -3.6945>, rotation=<rot 85.4723 0.0001 0.0001>}]
 pelvis [
  {shapeType=Capsule, length=10.6534, radius=9.6067, center=<0.175 -2.2729 0>}
  {shapeType=Capsule, length=14, radius=7.5, center=<-5.6175 -3.2328 0>}
 ]
 thigh_l [{shapeType=Capsule, length=27.75, radius=7.5, center=<-20.7286 -0.0003 0>, rotation=<rot 90 90 90.3127>}]
 thigh_r [{shapeType=Capsule, length=27.75, radius=7.5, center=<20.7286 0.0003 0>, rotation=<rot 90 -13.2613 167.0514>}]
}
constraints [
 {
  bone1Name=thigh_l
  bone2Name=pelvis
  swing1LimitDegrees=45
  swing1Motion=Free
  swing2LimitDegrees=45
  swing2Motion=Free
  twistLimitDegrees=45
  twistMotion=Free
 }
 {
  bone1Name=thigh_r
  bone2Name=pelvis
  swing1LimitDegrees=45
  swing1Motion=Free
  swing2LimitDegrees=45
  swing2Motion=Free
  twistLimitDegrees=45
  twistMotion=Free
 }
 {
  bone1Name=dyn_leg_pouch_base_r
  bone2Name=thigh_r
  swing1LimitDegrees=45
  swing1Motion=Free
  swing2LimitDegrees=45
  swing2Motion=Free
  twistLimitDegrees=45
  twistMotion=Free
 }
 {bone1Name=dyn_leg_pouch_r, bone2Name=dyn_leg_pouch_base_r, swing1LimitDegrees=5, swing2LimitDegrees=5, twistLimitDegrees=5}
 {bone1Name=dyn_leg_pouchStrap_r, bone2Name=dyn_leg_pouch_r, swing1LimitDegrees=45, swing2LimitDegrees=45, twistLimitDegrees=5}
 {
  bone1Name=dyn_back_pouch_base
  bone2Name=pelvis
  swing1LimitDegrees=45
  swing1Motion=Free
  swing2LimitDegrees=45
  swing2Motion=Free
  twistLimitDegrees=45
  twistMotion=Free
 }
 {bone1Name=dyn_back_pouch, bone2Name=dyn_back_pouch_base, swing1LimitDegrees=3, swing2LimitDegrees=3, twistLimitDegrees=3}
 {
  bone1Name=dyn_bk_strapA_a_base
  bone2Name=pelvis
  swing1LimitDegrees=45
  swing1Motion=Free
  swing2LimitDegrees=45
  swing2Motion=Free
  twistLimitDegrees=45
  twistMotion=Free
 }
 {bone1Name=dyn_bk_strapA_a, bone2Name=dyn_bk_strapA_a_base, swing1LimitDegrees=45, swing2LimitDegrees=45, twistLimitDegrees=5}
 {bone1Name=dyn_bk_strapA_b, bone2Name=dyn_bk_strapA_a, swing1LimitDegrees=45, swing2LimitDegrees=45, twistLimitDegrees=5}
 {bone1Name=dyn_bk_strapA_c, bone2Name=dyn_bk_strapA_b, swing1LimitDegrees=45, swing2LimitDegrees=45, twistLimitDegrees=5}
 {
  bone1Name=dyn_bk_strapB_a_base
  bone2Name=pelvis
  swing1LimitDegrees=45
  swing1Motion=Free
  swing2LimitDegrees=45
  swing2Motion=Free
  twistLimitDegrees=45
  twistMotion=Free
 }
 {bone1Name=dyn_bk_strapB_a, bone2Name=dyn_bk_strapB_a_base, swing1LimitDegrees=45, swing2LimitDegrees=25, twistLimitDegrees=5}
 {bone1Name=dyn_bk_strapB_b, bone2Name=dyn_bk_strapB_a, swing1LimitDegrees=45, swing2LimitDegrees=25, twistLimitDegrees=5}
 {bone1Name=dyn_bk_strapB_c, bone2Name=dyn_bk_strapB_b, swing1LimitDegrees=45, swing2LimitDegrees=25, twistLimitDegrees=5}
 {
  bone1Name=dyn_fr_beltStrapA_base_l
  bone2Name=pelvis
  swing1LimitDegrees=45
  swing1Motion=Free
  swing2LimitDegrees=45
  swing2Motion=Free
  twistLimitDegrees=45
  twistMotion=Free
 }
 {bone1Name=dyn_fr_beltStrapA_l, bone2Name=dyn_fr_beltStrapA_base_l, swing1LimitDegrees=10, swing2LimitDegrees=10, twistLimitDegrees=5}
 {bone1Name=dyn_fr_beltStrapB_l, bone2Name=dyn_fr_beltStrapA_l, swing1LimitDegrees=10, swing2LimitDegrees=10, twistLimitDegrees=5}
 {bone1Name=dyn_fr_beltStrapC_l, bone2Name=dyn_fr_beltStrapB_l, swing1LimitDegrees=10, swing2LimitDegrees=10, twistLimitDegrees=5}
 {
  bone1Name=dyn_glove_base_l
  bone2Name=pelvis
  swing1LimitDegrees=45
  swing1Motion=Free
  swing2LimitDegrees=45
  swing2Motion=Free
  twistLimitDegrees=45
  twistMotion=Free
 }
 {bone1Name=dyn_glove_l, bone2Name=dyn_glove_base_l, swing1LimitDegrees=4, swing2LimitDegrees=4, twistLimitDegrees=3}
 {
  bone1Name=dyn_glove_base_r
  bone2Name=pelvis
  swing1LimitDegrees=45
  swing1Motion=Free
  swing2LimitDegrees=45
  swing2Motion=Free
  twistLimitDegrees=45
  twistMotion=Free
 }
 {bone1Name=dyn_glove_r, bone2Name=dyn_glove_base_r, swing1LimitDegrees=4, swing2LimitDegrees=4, twistLimitDegrees=3}
]
